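# Physics asset for the BeachCavalier emote prop in CosmeticCompanions, defining simplified collision for its skeletal mesh.
# The bucket and firehydrant bones each get a body made of one capsule, with the hydrant's capsule the larger of the two.
# No constraints link the bodies, so the asset sets up no physical joint between them.
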PhysicsAsset /CosmeticCompanions/Assets/Props/Emotes/BeachCavalier/Meshes/PA_BeachCavalier
bodies {
 bucket [{shapeType=Capsule, length=16.821, radius=16.7649, center=<-1.2163 0.4359 -4.5988>, rotation=<rot -0.9511 -90 0.5695>}]
 firehydrant [{shapeType=Capsule, length=27.4903, radius=32.2337, center=<-2.5214 0.5212 25.3661>, rotation=<rot 1.0762 -90 0.499>}]
}
constraints []
